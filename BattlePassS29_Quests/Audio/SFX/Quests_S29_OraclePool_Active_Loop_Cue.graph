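# Sound cue for the active-state looping audio of the Oracle Pool in Season 29 Battle Pass quests.
# A single Wave Player loops the Quests_S29_OraclePool_Active_Loop_01 wave directly into the output, with no mixing or randomization.
# It plays in the quests sound class and reuses the supply-drop Hidden Star ambient loop attenuation for spatial falloff.
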SoundCue /BattlePassS29_Quests/Audio/SFX/Quests_S29_OraclePool_Active_Loop_Cue
root=WavePlayer_0
nodes [{id=WavePlayer_0, properties {bLooping=true, soundWaveAssetPtr=&/Quests_S29_OraclePool_Active_Loop_01}}]
properties {
 attenuationSettings=/Game/Athena/Sounds/SupplyDrops/Fort_HiddenStar_Ambient_Loop_Att
 maxDistance=2500
 soundClassObject=&Fort_Quests
}
