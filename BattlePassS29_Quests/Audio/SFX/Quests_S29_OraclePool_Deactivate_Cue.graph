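# Sound cue for deactivating the Oracle Pool in the Season 29 Battle Pass quests.
# A Random node picks one of three deactivate wave variations with equal weighting.
# It plays in the quests sound class with boosted volume, borrowing the attenuation asset from the Hidden Star supply drop ambient loop for spatial falloff.
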SoundCue /BattlePassS29_Quests/Audio/SFX/Quests_S29_OraclePool_Deactivate_Cue
root=Random_0
nodes [
 {id=Random_0, children [WavePlayer_1,WavePlayer_2,WavePlayer_3], properties {weights [1,1,1]}}
 {id=WavePlayer_1, properties {soundWaveAssetPtr=&/Quests_S29_OraclePool_Deactivate_01}}
 {id=WavePlayer_2, properties {soundWaveAssetPtr=&/Quests_S29_OraclePool_Deactivate_02}}
 {id=WavePlayer_3, properties {soundWaveAssetPtr=&/Quests_S29_OraclePool_Deactivate_03}}
]
properties {
 attenuationSettings=/Game/Athena/Sounds/SupplyDrops/Fort_HiddenStar_Ambient_Loop_Att
 maxDistance=2500
 soundClassObject=&Fort_Quests
 volumeMultiplier=1.6
}
