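# Looping sound cue for the Oracle Pool in the Battle Pass Season 29 quests, routed to the quests sound class at reduced volume.
# A single looping wave player feeds the output directly, with no randomization or mixing.
# Spatialization uses the shared ambient water pool attenuation asset; embedded attenuation overrides are stored but not switched on.
# A concurrency override allows one instance per owner and blocks new ones.
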SoundCue /BattlePassS29_Quests/Audio/SFX/Quests_S29_OraclePool_Loop_Cue
root=WavePlayer_0
nodes [{id=WavePlayer_0, properties {bLooping=true, soundWaveAssetPtr=&/Quests_S29_OraclePool_Loop_01}}]
properties {
 attenuationSettings=/Game/Athena/Sounds/Ambient/Fort_Amb_Water_Pool_Att
 bOverrideConcurrency=true
 maxDistance=2425
 soundClassObject=&Fort_Quests
 volumeMultiplier=0.4
 attenuationOverrides {
  bAttenuateWithLPF=true
  bEnableFocusInterpolation=true
  bEnableListenerFocus=true
  bEnableSubmixSends=true
  dBAttenuationAtMax=-28
  distanceAlgorithm=NaturalSound
  falloffDistance=46000
  focusReleaseInterpSpeed=0.3
  hPFFrequencyAtMax=200
  lPFFrequencyAtMax=1200
  lPFRadiusMax=60000
  lPFRadiusMin=0
  nonFocusAzimuth=90
  nonFocusVolumeAttenuation=0.5
  reverbWetLevelMax=1
  attenuationShapeExtents {x=1000}
  submixSendSettings [{
    bEnableHPFCutoff=false
    bEnableLPFCutoff=false
    disableManualSendClamp=false
    hPFCutoff=20
    lPFCutoff=20000
    maxSendDistance=3000
    maxSendLevel=0.3
    minSendDistance=0
    minSendLevel=0
    sendLevel=0.2
    sendLevelControlMethod=Linear
    soundSubmix=&Fortnite_Reverb_Distant
    customSendLevelCurve {externalCurve=/Game/Sounds/SubMixes/Reverb/Fortnite_Reverb_Distant_Send}
   }]
 }
 concurrencyOverrides {bLimitToOwner=true, maxCount=1, resolutionRule=PreventNew}
}
